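# Physics asset for the F_MED_Islander backpack cosmetic, providing collision and joint limits for its four dangling feather chains.
# The pack bone carries two large capsules as the body, while each feather chain has three thin capsule bodies, with the tip capsule longer.
# Each chain hangs from the pack and links bone to bone, with tight twist, wide swing, and a narrower second swing at the root joint.
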PhysicsAsset /BRCosmetics/Accessories/FORT_Backpacks/Backpack_F_MED_Islander/Meshes/F_MED_Islander_Pack_Physics
bodies {
 dyn_feathers_A_01 [{shapeType=Capsule, length=1, radius=1, center=<1 0 0>, rotation=<rot 90 0 0>}]
 dyn_feathers_A_02 [{shapeType=Capsule, length=1, radius=1, center=<1 0 0>, rotation=<rot 90 0 0>}]
 dyn_feathers_A_03 [{shapeType=Capsule, length=6.75, radius=1, center=<4.6 0 0>, rotation=<rot 90 0 0>}]
 dyn_feathers_B_01 [{shapeType=Capsule, length=1, radius=1, center=<1 0 0>, rotation=<rot 90 0 0>}]
 dyn_feathers_B_02 [{shapeType=Capsule, length=1, radius=1, center=<1 0 0>, rotation=<rot 90 0 0>}]
 dyn_feathers_B_03 [{shapeType=Capsule, length=6.75, radius=1, center=<5 0 0>, rotation=<rot 90 0 0>}]
 dyn_feathers_C_01 [{shapeType=Capsule, length=1, radius=1, center=<1 0 0>, rotation=<rot 90 0 0>}]
 dyn_feathers_C_02 [{shapeType=Capsule, length=1, radius=1, center=<1 0 0>, rotation=<rot 90 0 0>}]
 dyn_feathers_C_03 [{shapeType=Capsule, length=7, radius=1, center=<4.7 0 0>, rotation=<rot 90 0 0>}]
 dyn_feathers_D_01 [{shapeType=Capsule, length=1, radius=1, center=<1 0 0>, rotation=<rot 90 0 0>}]
 dyn_feathers_D_02 [{shapeType=Capsule, length=1, radius=1, center=<1 0 0>, rotation=<rot 90 0 0>}]
 dyn_feathers_D_03 [{shapeType=Capsule, length=5, radius=1, center=<3.6 0 0>, rotation=<rot 90 0 0>}]
 pack [
  {shapeType=Capsule, length=24.3975, radius=11.1301, center=<-9.1003 -6.3011 -5.3787>, rotation=<rot -15.8234 -89.9995 57.9296>}
  {shapeType=Capsule, length=15, radius=15, center=<-0.0001 16.6357 -16.913>, rotation=<rot 90 0 0>}
 ]
}
constraints [
 {bone1Name=dyn_feathers_A_01, bone2Name=pack, swing1LimitDegrees=75, swing2LimitDegrees=30, twistLimitDegrees=5}
 {bone1Name=dyn_feathers_A_02, bone2Name=dyn_feathers_A_01, swing1LimitDegrees=75, swing2LimitDegrees=75, twistLimitDegrees=5}
 {bone1Name=dyn_feathers_A_03, bone2Name=dyn_feathers_A_02, swing1LimitDegrees=75, swing2LimitDegrees=75, twistLimitDegrees=5}
 {bone1Name=dyn_feathers_B_01, bone2Name=pack, swing1LimitDegrees=75, swing2LimitDegrees=30, twistLimitDegrees=5}
 {bone1Name=dyn_feathers_B_02, bone2Name=dyn_feathers_B_01, swing1LimitDegrees=75, swing2LimitDegrees=75, twistLimitDegrees=5}
 {bone1Name=dyn_feathers_B_03, bone2Name=dyn_feathers_B_02, swing1LimitDegrees=75, swing2LimitDegrees=75, twistLimitDegrees=5}
 {bone1Name=dyn_feathers_C_01, bone2Name=pack, swing1LimitDegrees=75, swing2LimitDegrees=30, twistLimitDegrees=5}
 {bone1Name=dyn_feathers_C_02, bone2Name=dyn_feathers_C_01, swing1LimitDegrees=75, swing2LimitDegrees=75, twistLimitDegrees=5}
 {bone1Name=dyn_feathers_C_03, bone2Name=dyn_feathers_C_02, swing1LimitDegrees=75, swing2LimitDegrees=75, twistLimitDegrees=5}
 {bone1Name=dyn_feathers_D_01, bone2Name=pack, swing1LimitDegrees=75, swing2LimitDegrees=30, twistLimitDegrees=5}
 {bone1Name=dyn_feathers_D_02, bone2Name=dyn_feathers_D_01, swing1LimitDegrees=75, swing2LimitDegrees=75, twistLimitDegrees=5}
 {bone1Name=dyn_feathers_D_03, bone2Name=dyn_feathers_D_02, swing1LimitDegrees=75, swing2LimitDegrees=75, twistLimitDegrees=5}
]
